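# Material function for the 2025 dealership parent vehicle cosmetics material that applies ambient occlusion and curvature to an input color.
# The color is multiplied by the AO input, and AO_Intensity blends between the original and darkened result.
# That result is overlay-blended with the curvature input, and Curvature Intensity lerps between the AO-only and overlaid versions to produce the output.
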MaterialFunction /VehicleCosmetics/Materials/DealershipGeneralParent/MF_DealershipParent_2025_CurvAO
graph {nodes [
  n1 = FunctionInput() {inputName=InColor}
  n2 = LinearInterpolate(
   A=@n1
   B=Multiply(A=@n1, B=FunctionInput() {inputName=InAO, inputType=FunctionInput_Scalar, sortPriority=1})
   Alpha=FunctionInput() {inputName=AO_Intensity, inputType=FunctionInput_Scalar, sortPriority=3}
  ) {constAlpha=1}
  n3 = FunctionInput() {inputName=InCurv, inputType=FunctionInput_Scalar, sortPriority=2}
  FunctionOutput(None=LinearInterpolate(
    A=@n2
    B=MaterialFunctionCall(Base=@n2, Blend=@n3) {materialFunction=&Blend_Overlay}.Result
    Alpha=FunctionInput() {inputName=Curvature Intensity, inputType=FunctionInput_Scalar, sortPriority=4}
   )) {outputName=Lerp}
 ]}
