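# Material function for the dealership vehicle cosmetic parent material that adds optional glitter sparkles to base color and emissive.
# Multiplies two samples of one noise texture, one on tiled UVs and one on an offset reflection vector, so glints are view-dependent.
# Scales and floors the product into a hard-edged sparkle mask, trimming aliased artifacts.
# Builds an area mask from selectable color-mask channels, with lerps controlling the final blend.
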MaterialFunction /VehicleCosmetics/Materials/DealershipGeneralParent/MF_DealershipParent_2025_Sparkle
graph {nodes [
  FunctionOutput(None=NamedRerouteUsage()) {outputName=BC_OUT}
  FunctionOutput(None=NamedRerouteUsage()) {outputName=Emissive_OUT, sortPriority=10}
  n1 = ReflectionVectorWS()
  n2 = ComponentMask(None=Add(A=Transform(None=@n1), B=NamedRerouteUsage())) {g=true, r=true}
  n3 = TextureCoordinate()
  n4 = VectorParameter() {
   desc="Higher tiling values == denser sparkles."
   group=05_Sparkles
   parameterName=SparkleTiling
   sortPriority=41
   defaultValue=<rgba 4 4 0 1>
  }
  n5 = AppendVector(A=@n4.R, B=@n4.R)
  n6 = Multiply(
   A=TextureSample(UVs=Multiply(A=@n2) {constB=0.0625}) {samplerSource=SSM_Wrap_WorldGroupSettings, texture=/VehicleCosmetics/Textures/UniversalTextures/T_Noise_Generic_01_RGB}.B
   B=TextureSample(UVs=Multiply(A=@n3, B=@n5)) {samplerSource=SSM_Wrap_WorldGroupSettings, texture=&T_Noise_Generic_01_RGB}.B
  )
  n7 = ScalarParameter() {defaultValue=2, desc="Lower values == fewer standout sparkles.", group=05_Sparkles, parameterName=SparkleIntensity, sortPriority=49}
  n8 = Multiply(A=@n7) {constB=2048}
  n9 = Floor(None=Add(A=Multiply(A=@n6, B=@n8) {constB=4096}) {constB=-0.03, desc=Darken/Smooth aliased edges}) {desc=Further cleaning glitter artifacting}
  NamedRerouteDeclaration(None=@n9) {name=SparkleMaskEmis_Out_1, nodeColor=<rgba 0.1016 1 0 1>}
  NamedRerouteDeclaration(None=FunctionInput() {inputName=BaseColorComposite, previewValue=<1 1 1 1>}) {name=BCBase_Out_1, nodeColor=<rgba 0.0095 0 1 1>}
  NamedRerouteDeclaration(None=FunctionInput() {inputName=NormalTexture, sortPriority=10}) {name=Normal_Out_1, nodeColor=<rgba 0.509 0.1352 1 1>}
  NamedRerouteDeclaration(None=FunctionInput() {inputName=EmissiveColor, sortPriority=20, previewValue=<0 0 1 1>}) {name=EmissiveBase_Out_1, nodeColor=<rgba 0.8209 1 0 1>}
  n10 = NamedRerouteUsage()
  n11 = LinearInterpolate(A=@n10, B=NamedRerouteUsage(), Alpha=NamedRerouteUsage())
  n12 = Saturate(None=NamedRerouteUsage())
  NamedRerouteDeclaration(None=LinearInterpolate(A=@n10, B=LinearInterpolate(A=@n10, B=@n11, Alpha=@n12), Alpha=ScalarParameter() {group=05_Sparkles, parameterName=Sparkles, sortPriority=0})) {name=Emissive_Out_1, nodeColor=<rgba 0 1 0.2266 1>}
  n13 = FunctionInput() {inputName=ColorMaskTexture, inputType=FunctionInput_Vector4, sortPriority=30, previewValue=<1 1 1 1>}
  n14 = VectorParameter() {group=05_Sparkles, parameterName=Sparkle_UseMaskChannel, sortPriority=1, defaultValue=<rgba 1 1 1 1>}
  n15 = MaterialFunctionCall!(Float4=Multiply(A=@n13, B=@n14.RGBA)) {materialFunction=/Game/Global/MaterialFunctions/MF_BreakComponents}
  n16 = Max(A=@n15.R, B=@n15.R)
  NamedRerouteDeclaration(None=Max(A=Max(A=@n16, B=@n15.B), B=@n15.A)) {name=RGBMask_Out_1, nodeColor=<rgba 1 0.0447 0 1>}
  n17 = NamedRerouteUsage()
  n18 = NamedRerouteUsage()
  n19 = Saturate(None=NamedRerouteUsage())
  n20 = NamedRerouteUsage()
  NamedRerouteDeclaration(None=LinearInterpolate(A=@n20, B=LinearInterpolate(A=@n17, B=@n18, Alpha=@n19), Alpha=ScalarParameter() {group=05_Sparkles, parameterName=Sparkles})) {name=BCSparkle_Out_1, nodeColor=<rgba 0 1 0.2266 1>}
  n21 = Saturate(None=@n9)
  NamedRerouteDeclaration(None=Add(A=Multiply(A=@n21, B=@n20), B=@n20)) {name=SparkleBC_Out_1, nodeColor=<rgba 0 0.4375 1 1>}
 ]}
