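# Parent material function for dealership vehicle cosmetics, building full material attributes from base color, normal, SRM (specular/roughness/metallic) and mask textures.
# Four color-mask channels layer toggleable color and SRM overrides over texture or vector bases, with AO/curvature tinting through CurvAO and per-channel emissive.
# A Sparkle subfunction processes color and emissive.
# A static switch enables clear coat with its own bottom normal, mask-blended against default shading; sampled textures are also output.
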MaterialFunction /VehicleCosmetics/Materials/DealershipGeneralParent/MF_DealershipParent_2025
graph {nodes [
  n1 = MaterialFunctionCall(Float3=NamedRerouteUsage()) {materialFunction=&BreakOutFloat3Components}
  n2 = NamedRerouteUsage()
  n3 = MaterialFunctionCall!(BaseColorComposite=NamedRerouteUsage(), NormalTexture=@n2, EmissiveColor=NamedRerouteUsage(), ColorMaskTexture=NamedRerouteUsage()) {materialFunction=&/MF_DealershipParent_2025_Sparkle}
  n4 = SetMaterialAttributes!(
   MaterialAttributes
   BaseColor=@n3.BC_OUT
   Specular=@n1.R
   Roughness=@n1.R
   Metallic=@n1.R
   Normal=@n2
   EmissiveColor=@n3.Emissive_OUT
   ShadingModel=ShadingModel()
  ) {attributeSetTypes [
    #1
    #2
    #3
    #4
    #5
    #6
    #7
   ]}
  n5 = TextureSampleParameter2D(UVs=Multiply(A=NamedRerouteUsage(), B=ScalarParameter() {defaultValue=1, group=03_ClearCoat, parameterName=BottomNormalTiling})) {group=03_ClearCoat, parameterName=BottomNormal, samplerType=SAMPLERTYPE_Normal, texture=&FlatNormalBlue}
  n6 = ScalarParameter() {group=03_ClearCoat, parameterName=BottomNormalFlatness}
  n7 = SetMaterialAttributes!(
   MaterialAttributes=@n4
   CustomData0=Saturate(None=ScalarParameter() {defaultValue=1, group=03_ClearCoat, parameterName=ClearCoat})
   CustomData1=Saturate(None=ScalarParameter() {defaultValue=0.01, group=03_ClearCoat, parameterName=ClearCoatRoughness})
   ClearCoatBottomNormal=NamedRerouteDeclaration(None=MaterialFunctionCall(Normal=@n5.RGB, Flatness=@n6) {materialFunction=&FlattenNormal}.Result) {name=CCBottomNormal, nodeColor=<rgba 1 0.6562 0 1>}
   ShadingModel=ShadingModel() {shadingModel=MSM_ClearCoat}
  ) {attributeSetTypes [
    #8
    #9
    #10
    #7
   ]}
  n8 = GetMaterialAttributes(None=@n7) {attributeGetTypes [#5, #10]}
  n9 = SetMaterialAttributes!(
   MaterialAttributes=@n8.MaterialAttributes
   ShadingModel=ShadingModel()
   Normal=MaterialFunctionCall(BaseNormal=@n8.Normal, AdditionalNormal=@n8.Normal) {materialFunction=&BlendAngleCorrectedNormals}.Result
  ) {attributeSetTypes [#7, #5]}
  n10 = Saturate(None=Ceil(None=NamedRerouteUsage()))
  FunctionOutput(None=StaticSwitchParameter(True=BlendMaterialAttributes(A=@n9, B=@n7, Alpha=@n10), False=@n4) {group=03_ClearCoat, parameterName=UseClearCoatShading})
  FunctionOutput(None=NamedRerouteUsage()) {outputName=CC_BottomNormal_Texture}
  n11 = TextureObjectParameter(UVs=FunctionInput() {inputName=Masks}) {
   group=00_Texture Inputs
   parameterName=Masks
   samplerType=SAMPLERTYPE_Masks
   sortPriority=36
   texture=&T_BlackMask
   channelNames {a=EmissiveMask, b=ClearCoat, g=Color Overlay, r=Color Multiply}
  }
  n12 = TextureSample(UVs=Multiply(A=NamedRerouteUsage(), B=ScalarParameter() {defaultValue=1, group=04_Tiling, parameterName=MaskTextureTiling}), Tex=@n11) {samplerType=SAMPLERTYPE_Masks, texture=&T_BlackMask}
  FunctionOutput(None=@n12.RGBA) {outputName=Mask_Texture}
  n13 = TextureObjectParameter(UVs=FunctionInput() {inputName=ColorMasks}) {
   group=02_Color Masks
   parameterName=ColorMask
   samplerType=SAMPLERTYPE_Masks
   sortPriority=-2
   texture=&T_BlackMask
   channelNames {a=ColorA, b=ColorB, g=ColorG, r=ColorR}
  }
  n14 = TextureSample(UVs=Multiply(A=NamedRerouteUsage(), B=ScalarParameter() {defaultValue=1, group=04_Tiling, parameterName=ColorMaskTextureTiling}), Tex=@n13) {
   samplerType=SAMPLERTYPE_Masks
   texture=&T_BlackMask
   channelNames {a=ColorA, b=ColorB, g=ColorG, r=ColorR}
  }
  FunctionOutput(None=NamedRerouteDeclaration(None=@n14.RGBA) {name=ColorMaskTexture, nodeColor=<rgba 1 0 1 1>}) {outputName=ColorMask_Texture}
  n15 = Multiply(A=NamedRerouteUsage(), B=ScalarParameter() {defaultValue=1, group=04_Tiling, parameterName=BaseColorTiling})
  n16 = TextureObjectParameter(UVs=FunctionInput() {inputName=BaseColor}) {group=00_Texture Inputs, parameterName=BaseColor, samplerType=SAMPLERTYPE_Masks, texture=&T_BlackMask}
  FunctionOutput(None=NamedRerouteDeclaration(None=TextureSample(UVs=@n15, Tex=@n16) {samplerType=SAMPLERTYPE_Masks, texture=&T_BlackMask}.RGB) {name=BaseColorTexture, nodeColor=<rgba 1 0 1 1>}) {outputName=BaseColor_Texture}
  n17 = Multiply(A=NamedRerouteUsage(), B=ScalarParameter() {defaultValue=1, group=04_Tiling, parameterName=NormalTiling})
  n18 = TextureObjectParameter(UVs=FunctionInput() {inputName=Normal}) {group=00_Texture Inputs, parameterName=Normal, samplerType=SAMPLERTYPE_Normal, texture=&FlatNormalBlue}
  FunctionOutput(None=NamedRerouteDeclaration(None=TextureSample(UVs=@n17, Tex=@n18) {samplerType=SAMPLERTYPE_Masks, texture=&T_BlackMask}.RGB) {name=NormalTexture, nodeColor=<rgba 1 0 1 1>}) {outputName=Normal_Texture}
  n19 = TextureObjectParameter(UVs=FunctionInput() {inputName=SRM}) {
   group=00_Texture Inputs
   parameterName=SRM
   samplerType=SAMPLERTYPE_LinearColor
   texture=/Game/Global/Textures/Default/Blanks/T_BlackLinearColor
   channelNames {a=---, b=Metallic, g=Roughness, r=Specular}
  }
  n20 = TextureSample(UVs=Multiply(A=NamedRerouteUsage(), B=ScalarParameter() {defaultValue=1, group=04_Tiling, parameterName=SRMTiling}), Tex=@n19) {
   samplerType=SAMPLERTYPE_Normal
   texture=&FlatNormalBlue
   channelNames {a=--, b=Metallic, g=Roughness, r=Specular}
  }
  FunctionOutput(None=NamedRerouteDeclaration(None=@n20.RGB) {name=SRMTexture, nodeColor=<rgba 1 0 1 1>}) {outputName=SRM_Texture}
  NamedRerouteDeclaration(None=Multiply(A=ScalarParameter() {defaultValue=1, group=04_Tiling, parameterName=GlobalTiling, sortPriority=0}, B=TextureCoordinate())) {name=UVs, nodeColor=<rgba 0.125 1 0 1>}
  n21 = MaterialFunctionCall!(
   InColor=NamedRerouteUsage()
   InAO=NamedRerouteUsage()
   InCurv=NamedRerouteUsage()
   AO_Intensity=NamedRerouteUsage()
   Curvature Intensity=NamedRerouteUsage()
  ) {materialFunction=&/MF_DealershipParent_2025_CurvAO}
  n22 = Multiply(A=@n21.Lerp, B=NamedRerouteUsage())
  n23 = NamedRerouteUsage() {desc="Allows for using a vector parameter for the basecolor output. 0 means you'll use the BC texture as the bottom layer of this mat. 1 means you can use the \"BaseColorVector\" parameter below"}
  n24 = Saturate(None=Ceil(None=@n23))
  n25 = Multiply(A=@n24, B=NamedRerouteUsage())
  n26 = MaterialFunctionCall(
   InColor=NamedRerouteUsage()
   InAO=NamedRerouteUsage()
   InCurv=NamedRerouteUsage()
   AO_Intensity=NamedRerouteUsage()
   Curvature Intensity=NamedRerouteUsage()
  ) {materialFunction=&MF_DealershipParent_2025_CurvAO}
  n27 = NamedRerouteUsage()
  n28 = Multiply(
   A=ComponentMask(None=NamedRerouteUsage()) {r=true}
   B=ChannelMaskParameter(None=NamedRerouteUsage()) {group=05_Tech Art, parameterName=ColorRMaskChannel, sortPriority=0}.R
  )
  n29 = NamedRerouteUsage()
  n30 = MaterialFunctionCall(
   InColor=NamedRerouteUsage()
   InAO=NamedRerouteUsage()
   InCurv=NamedRerouteUsage()
   AO_Intensity=NamedRerouteUsage()
   Curvature Intensity=NamedRerouteUsage()
  ) {materialFunction=&MF_DealershipParent_2025_CurvAO}
  n31 = ChannelMaskParameter(None=NamedRerouteUsage()) {group=05_Tech Art, maskChannel=Green, parameterName=ColorGMaskChannel, sortPriority=1, defaultValue=<rgba 0 1 0 0>}
  n32 = Multiply(A=ComponentMask(None=NamedRerouteUsage()) {g=true}, B=@n31.R)
  n33 = LinearInterpolate(
   A=LinearInterpolate(
    A=LinearInterpolate(B=@n22, Alpha=@n25) {desc=BC Texture | BC Vector Emissive}
    B=Multiply(A=@n26.Lerp, B=@n27)
    Alpha=Multiply(A=@n28, B=@n29)
   ) {desc=ColorR Emissive}
   B=Multiply(A=@n30.Lerp, B=NamedRerouteUsage())
   Alpha=Multiply(A=@n32, B=NamedRerouteUsage())
  )
  n34 = MaterialFunctionCall(
   InColor=NamedRerouteUsage()
   InAO=NamedRerouteUsage()
   InCurv=NamedRerouteUsage()
   AO_Intensity=NamedRerouteUsage()
   Curvature Intensity=NamedRerouteUsage()
  ) {materialFunction=&MF_DealershipParent_2025_CurvAO}
  n35 = Multiply(A=@n34.Lerp, B=NamedRerouteUsage())
  n36 = ChannelMaskParameter(None=NamedRerouteUsage()) {group=05_Tech Art, maskChannel=Blue, parameterName=ColorBMaskChannel, sortPriority=2, defaultValue=<rgba 0 0 1 0>}
  n37 = Multiply(A=ComponentMask(None=NamedRerouteUsage()) {b=true}, B=@n36.R)
  n38 = Multiply(A=@n37, B=NamedRerouteUsage())
  n39 = MaterialFunctionCall(
   InColor=NamedRerouteUsage()
   InAO=NamedRerouteUsage()
   InCurv=NamedRerouteUsage()
   AO_Intensity=NamedRerouteUsage()
   Curvature Intensity=NamedRerouteUsage()
  ) {materialFunction=&MF_DealershipParent_2025_CurvAO}
  n40 = NamedRerouteUsage()
  n41 = ChannelMaskParameter(None=NamedRerouteUsage()) {group=05_Tech Art, maskChannel=Alpha, parameterName=ColorAMaskChannel, sortPriority=3, defaultValue=<rgba 0 0 0 1>}
  n42 = Multiply(A=ComponentMask(None=NamedRerouteUsage()) {a=true}, B=@n41.R)
  n43 = NamedRerouteUsage()
  NamedRerouteDeclaration(None=LinearInterpolate(A=LinearInterpolate(A=@n33, B=@n35, Alpha=@n38), B=Multiply(A=@n39.Lerp, B=@n40), Alpha=Multiply(A=@n42, B=@n43))) {name=EmissiveColor, nodeColor=<rgba 1 0.5859 0 1>}
  NamedRerouteDeclaration(None=@n12.A) {name=EmissiveMask, nodeColor=<rgba 1 0 1 1>}
  NamedRerouteDeclaration(None=@n12.B) {name=ClearCoat, nodeColor=<rgba 1 0 1 1>}
  n44 = VectorParameter() {
   group=01_Base
   parameterName=Base_Settings
   sortPriority=0
   channelNames {a=Base Overlay Intensity, b=Base Multiply Intensity, g=Use Base SRM, r=Use Base Color}
  }
  NamedRerouteDeclaration(None=@n44.Use Base Color) {name=BCVectorSettings_UseVectorColor, nodeColor=<rgba 0.2 0.2 0.2 1>}
  NamedRerouteDeclaration(None=@n44.Use Base SRM) {name=BCVectorSettings_UseBaseColorSRM, nodeColor=<rgba 0.2 0.2 0.2 1>}
  NamedRerouteDeclaration(None=@n44.Base Multiply Intensity) {name=BCVectorSettings_AoIntensity, nodeColor=<rgba 0.2 0.2 0.2 1>}
  NamedRerouteDeclaration(None=@n44.Base Overlay Intensity) {name=BCVectorSettings_CurvIntensity, nodeColor=<rgba 0.2 0.2 0.2 1>}
  n45 = VectorParameter() {group=01_Base, parameterName=Base_Color, channelNames {a=Emissive Intensity, b="", g="", r=""}, defaultValue=<rgba 0.5 0.5 0.5 0>}
  NamedRerouteDeclaration(None=@n45) {name=BCVector_Color, nodeColor=<rgba 0.2 0.2 0.2 1>}
  NamedRerouteDeclaration(None=@n45.Emissive Intensity) {name=BCVector_EmissiveIntensity, nodeColor=<rgba 0.2 0.2 0.2 1>}
  n46 = VectorParameter() {group=01_Base, parameterName=Base_SRM, channelNames {a="", b=Metallic, g=Roughness, r=Specular}, defaultValue=<rgba 0.5 0.5 0.5 1>}
  NamedRerouteDeclaration(None=@n46) {name=BCVector_SRM, nodeColor=<rgba 0.2 0.2 0.2 1>}
  NamedRerouteDeclaration(None=VectorParameter() {group=02_Color Masks, parameterName=ColorR, sortPriority=0, defaultValue=<rgba 1 0 0 0>}) {name=ColorR, nodeColor=<rgba 1 0 0 1>}
  n47 = VectorParameter() {
   group=02_Color Masks
   parameterName=ColorR_Settings
   sortPriority=0
   channelNames {a="", b=Emissive Intensity, g=Overlay Intensity, r=Multiply Intensity}
  }
  NamedRerouteDeclaration(None=@n47.Emissive Intensity) {name=ColorR_EmissiveIntensity, nodeColor=<rgba 1 0 0 1>}
  NamedRerouteDeclaration(None=@n47.Multiply Intensity) {name=ColorR_AOIntensity, nodeColor=<rgba 1 0 0 1>}
  NamedRerouteDeclaration(None=@n47.Overlay Intensity) {name=ColorR_CurveIntensity, nodeColor=<rgba 1 0 0 1>}
  n48 = VectorParameter() {
   group=02_Color Masks
   parameterName=ColorG_Settings
   sortPriority=1
   channelNames {a="", b=Emissive Intensity, g=Overlay Intensity, r=Multiply Intensity}
  }
  NamedRerouteDeclaration(None=@n48.Emissive Intensity) {name=ColorG_EmissiveIntensity, nodeColor=<rgba 0 1 0 1>}
  NamedRerouteDeclaration(None=VectorParameter() {group=02_Color Masks, parameterName=ColorB, sortPriority=2, defaultValue=<rgba 0 0 1 0>}) {name=ColorB, nodeColor=<rgba 0 0 1 1>}
  n49 = VectorParameter() {
   group=02_Color Masks
   parameterName=ColorB_Settings
   sortPriority=2
   channelNames {a="", b=Emissive Intensity, g=Overlay Intensity, r=Multiply Intensity}
  }
  NamedRerouteDeclaration(None=@n49.Emissive Intensity) {name=ColorB_EmissiveIntensity, nodeColor=<rgba 0 0 1 1>}
  NamedRerouteDeclaration(None=@n49.Multiply Intensity) {name=ColorB_AOIntensity, nodeColor=<rgba 0 0 1 1>}
  NamedRerouteDeclaration(None=@n49.Overlay Intensity) {name=ColorB_CurvatureIntensity, nodeColor=<rgba 0 0 1 1>}
  NamedRerouteDeclaration(None=VectorParameter() {group=02_Color Masks, parameterName=ColorA, sortPriority=3, defaultValue=<rgba 1 1 1 0>}) {name=ColorA, nodeColor=<rgba 1 1 1 1>}
  n50 = VectorParameter() {
   group=02_Color Masks
   parameterName=ColorA_Settings
   sortPriority=3
   channelNames {a="", b=Emissive Intensity, g=Overlay Intensity, r=Multiply Intensity}
  }
  NamedRerouteDeclaration(None=@n50.Emissive Intensity) {name=ColorA_EmissiveIntensity, nodeColor=<rgba 1 1 1 1>}
  NamedRerouteDeclaration(None=@n50.Multiply Intensity) {name=ColorA_AOIntensity, nodeColor=<rgba 1 1 1 1>}
  NamedRerouteDeclaration(None=@n50.Overlay Intensity) {name=ColorA_CurveIntensity, nodeColor=<rgba 1 1 1 1>}
  n51 = VectorParameter() {
   group=02_Color Masks
   parameterName=ColorR_SRM
   sortPriority=0
   channelNames {a="", b=Metallic, g=Roughness, r=Specular}
   defaultValue=<rgba 0.5 0.5 0 1>
  }
  NamedRerouteDeclaration(None=@n51) {name=ColorR_SRM, nodeColor=<rgba 1 0 0 1>}
  n52 = VectorParameter() {
   group=02_Color Masks
   parameterName=ColorG_SRM
   sortPriority=1
   channelNames {a="", b=Metallic, g=Roughness, r=Specular}
   defaultValue=<rgba 0.5 0.5 0 1>
  }
  NamedRerouteDeclaration(None=@n52) {name=ColorG_SRM, nodeColor=<rgba 0 1 0 1>}
  n53 = VectorParameter() {
   group=02_Color Masks
   parameterName=ColorB_SRM
   sortPriority=2
   channelNames {a="", b=Metallic, g=Roughness, r=Specular}
   defaultValue=<rgba 0.5 0.5 0 1>
  }
  NamedRerouteDeclaration(None=@n53) {name=ColorB_SRM, nodeColor=<rgba 0 0 1 1>}
  n54 = VectorParameter() {
   group=02_Color Masks
   parameterName=ColorA_SRM
   sortPriority=3
   channelNames {a="", b=Metallic, g=Roughness, r=Specular}
   defaultValue=<rgba 0.5 0.5 0 1>
  }
  NamedRerouteDeclaration(None=@n54) {name=ColorA_SRM, nodeColor=<rgba 1 1 1 1>}
  n55 = VectorParameter() {
   desc=Toggle the color on or off for the selected channel. uses the ColorMask texture
   group=02_Color Masks
   parameterName=ColorMask_UseColor
   sortPriority=-1
   channelNames {a=Toggle Color A, b=Toggle Color B, g=Toggle Color G, r=Toggle Color R}
   defaultValue=<rgba 1 1 1 0>
  }
  n56 = AppendVector(A=@n55, B=@n55)
  NamedRerouteDeclaration(None=Saturate(None=Ceil(None=@n56))) {name=UseColorVector, nodeColor=<rgba 0.5 0 1 1>}
  NamedRerouteDeclaration(None=VectorParameter() {group=02_Color Masks, parameterName=ColorG, sortPriority=1, defaultValue=<rgba 0 1 0 0>}) {name=ColorG, nodeColor=<rgba 0 1 0 1>}
  NamedRerouteDeclaration(None=@n48.Multiply Intensity) {name=ColorG_AOIntensity, nodeColor=<rgba 0 1 0 1>}
  NamedRerouteDeclaration(None=@n48.Overlay Intensity) {name=ColorG_CurveIntensity, nodeColor=<rgba 0 1 0 1>}
  NamedRerouteDeclaration(None=@n12.R) {name=AO, nodeColor=<rgba 1 0 1 1>}
  NamedRerouteDeclaration(None=@n12.G) {name=Curvature, nodeColor=<rgba 1 0 1 1>}
  n57 = VectorParameter() {
   desc=Toggles the SRM values on or off for the selected channel. Uses the ColorMask Texture. Off defaults to the SRM texture.
   group=02_Color Masks
   parameterName=ColorMask_UseColorSRM
   sortPriority=-1
   channelNames {a=ColorA Use Vector SRM, b=ColorB Use Vector SRM, g=ColorG Use Vector SRM, r=ColorR Use Vector SRM}
  }
  n58 = AppendVector(A=@n57, B=@n57)
  NamedRerouteDeclaration(None=Saturate(None=Ceil(None=@n58))) {name=UseColorSRMVector, nodeColor=<rgba 0.5 0 1 1>}
  n59 = VectorParameter() {
   group=01_Base
   parameterName=BaseColorTexture_Settings
   sortPriority=40
   channelNames {a="", b="", g=BaseColorTexture Curvature Intensity, r=BaseColorTexture Multiply Intensity}
  }
  n60 = MaterialFunctionCall(
   InColor=NamedRerouteUsage()
   InAO=NamedRerouteUsage()
   InCurv=NamedRerouteUsage()
   AO_Intensity=@n59.BaseColorTexture Multiply Intensity
   Curvature Intensity=@n59.BaseColorTexture Multiply Intensity
  ) {materialFunction=&MF_DealershipParent_2025_CurvAO}
  n61 = LinearInterpolate(
   A=LinearInterpolate(A=LinearInterpolate(A=@n60.Lerp, B=@n21.Lerp, Alpha=@n24) {desc="BC Texture | BC Vector "}, B=@n26.Lerp, Alpha=@n28) {desc=ColorR}
   B=@n30.Lerp
   Alpha=@n32
  )
  NamedRerouteDeclaration(None=LinearInterpolate(A=LinearInterpolate(A=@n61, B=@n34.Lerp, Alpha=@n37), B=@n39.Lerp, Alpha=@n42)) {name=BaseColorComposite, nodeColor=<rgba 0 0.7891 1 1>}
  n62 = NamedRerouteUsage()
  n63 = NamedRerouteUsage()
  n64 = NamedRerouteUsage() {desc="The parameter allows you to use the \"BaseColorVectorSRM\" below if set to 1. If set to 0 that means you'll use the SRM texture instead."}
  n65 = Saturate(None=Ceil(None=@n64))
  n66 = NamedRerouteUsage()
  n67 = Saturate(None=Ceil(None=ComponentMask(None=@n66) {r=true}))
  n68 = ChannelMaskParameter(None=NamedRerouteUsage()) {group=06_Tech Art, parameterName=ColorRMaskChannel, sortPriority=0}
  n69 = Ceil(None=ComponentMask(None=NamedRerouteUsage()) {g=true})
  n70 = ChannelMaskParameter(None=NamedRerouteUsage()) {group=06_Tech Art, maskChannel=Green, parameterName=ColorGMaskChannel, sortPriority=1, defaultValue=<rgba 0 1 0 0>}
  n71 = LinearInterpolate(
   A=LinearInterpolate(A=LinearInterpolate(A=@n62, B=@n63, Alpha=@n65), B=NamedRerouteUsage(), Alpha=Multiply(A=@n67, B=@n68.R))
   B=NamedRerouteUsage()
   Alpha=Multiply(A=Saturate(None=@n69), B=@n70.R)
  )
  n72 = NamedRerouteUsage()
  n73 = ComponentMask(None=NamedRerouteUsage()) {b=true}
  n74 = ChannelMaskParameter(None=NamedRerouteUsage()) {group=06_Tech Art, maskChannel=Blue, parameterName=ColorBMaskChannel, sortPriority=2, defaultValue=<rgba 0 0 1 0>}
  n75 = Multiply(A=Saturate(None=Ceil(None=@n73)), B=@n74.R)
  n76 = NamedRerouteUsage()
  n77 = Saturate(None=Ceil(None=ComponentMask(None=@n76) {a=true}))
  n78 = ChannelMaskParameter(None=NamedRerouteUsage()) {group=06_Tech Art, maskChannel=Alpha, parameterName=ColorAMaskChannel, sortPriority=3, defaultValue=<rgba 0 0 0 1>}
  NamedRerouteDeclaration(None=LinearInterpolate(A=LinearInterpolate(A=@n71, B=@n72, Alpha=@n75), B=NamedRerouteUsage(), Alpha=Multiply(A=@n77, B=@n78.R))) {name=SRMComposite, nodeColor=<rgba 1 0.8906 0 1>}
 ]}
