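# Parent material for 2025 dealership vehicle cosmetics, set up for skeletal meshes and mesh deformers.
# MF_DealershipParent_2025 builds the base material attributes, which pass through MF_BaseVehicle_Damage and then MF_VehicleDeRez_Transition.
# Shading model comes from the graph (Default Lit or Clear Coat), taken before the damage function and reapplied after it.
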Material /VehicleCosmetics/Materials/DealershipGeneralParent/M_Dealership_Parent_2025
graph {
 properties {
  bAutomaticallySetUsageInEditor=false
  bUseMaterialAttributes=true
  bUsedWithMeshDeformer=true
  bUsedWithSkeletalMesh=true
  shadingModel=MSM_FromMaterialExpression
  usedShadingModels=Default Lit | Clear Coat
 }
 nodes [
  n1 = GetMaterialAttributes(None=MaterialFunctionCall!(ColorMasks, BaseColor, Normal, SRM, Masks) {materialFunction=&/MF_DealershipParent_2025}.Result) {attributeGetTypes [#1]}
  MaterialFunctionCall(In=SetMaterialAttributes!(
    MaterialAttributes=MaterialFunctionCall(Material=@n1.MaterialAttributes) {materialFunction=&MF_BaseVehicle_Damage}.Result
    Shading Model=@n1.Shading Model
   ) {attributeSetTypes [#1]}) {materialFunction=&MF_VehicleDeRez_Transition}
 ]
}
